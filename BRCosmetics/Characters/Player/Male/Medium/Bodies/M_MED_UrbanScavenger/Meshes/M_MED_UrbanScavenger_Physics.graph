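# Physics asset for the medium male UrbanScavenger player body, simulating secondary motion of its arm D-rings, bandolier, pouches and scarf flaps.
# A spine_02 capsule is the only torso body, linked to a small anchor box per accessory by angularly free constraints.
# Each anchor drives a box body through tight angular limits.
# D-rings lock twist and one swing to act as hinges, and scarf flap B forms a two-link chain.
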PhysicsAsset /BRCosmetics/Characters/Player/Male/Medium/Bodies/M_MED_UrbanScavenger/Meshes/M_MED_UrbanScavenger_Physics
bodies {
 dyn_armDring_Base_l [{shapeType=Box, extentX=1.01, extentY=1.01, extentZ=1.01, rotation=<rot 0 -90 0>}]
 dyn_armDring_Base_r [{shapeType=Box, extentX=1.01, extentY=1.01, extentZ=1.01, rotation=<rot 0 -90 0>}]
 dyn_armDring_l [{shapeType=Box, extentX=0.7742, extentY=3.2051, extentZ=3.3082, center=<0.6333 0.0171 0.0215>, rotation=<rot -5.5845 -89.9999 1.9591>}]
 dyn_armDring_r [{shapeType=Box, extentX=0.7783, extentY=3.3075, extentZ=3.3082, center=<0.6204 0.0148 -0.02>, rotation=<rot 5.758 -89.9999 -2.1012>}]
 dyn_bandolier [{shapeType=Box, extentX=2.4709, extentY=6.5808, extentZ=10.5102, center=<4.5451 0.1628 0.1199>, rotation=<rot 0 -91.9983 -89.3859>}]
 dyn_bandolier_Base [{shapeType=Box, extentX=1.01, extentY=1.01, extentZ=1.01, rotation=<rot 0 -90 0>}]
 dyn_pouchA_Base_l [{shapeType=Box, extentX=1.01, extentY=1.01, extentZ=1.01, rotation=<rot 0 -90 0>}]
 dyn_pouchA_l [{shapeType=Box, extentX=3.7152, extentY=5.489, extentZ=12.5365, center=<2.2021 -0.8857 0.19>, rotation=<rot 0 102.9989 -89.5916>}]
 dyn_pouchB_Base_l [{shapeType=Box, extentX=1.01, extentY=1.01, extentZ=1.01, rotation=<rot 0 -90 0>}]
 dyn_pouchB_l [{shapeType=Box, extentX=4.3484, extentY=5.6698, extentZ=12.5365, center=<2.4249 -0.8151 0.0361>, rotation=<rot 0 103.013 -87.5305>}]
 dyn_pouchC_Base_l [{shapeType=Box, extentX=1.01, extentY=1.01, extentZ=1.01, rotation=<rot 0 -90 0>}]
 dyn_pouchC_l [{shapeType=Box, extentX=3.7705, extentY=5.5255, extentZ=12.5365, center=<2.3937 -0.9641 -0.0693>, rotation=<rot 0 103.8029 -88.7964>}]
 dyn_pouch_Base_r [{shapeType=Box, extentX=1.01, extentY=1.01, extentZ=1.01, rotation=<rot 0 -90 0>}]
 dyn_pouch_r [{shapeType=Box, extentX=5.9052, extentY=8.7978, extentZ=14.0506, center=<3.3274 -2.1123 0.1827>, rotation=<rot 0 -85.2849 -89.7339>}]
 dyn_scarfFlapA [{shapeType=Box, extentX=1.091, extentY=6.5635, extentZ=12.8008, center=<5.8514 0.1128 0.1614>, rotation=<rot 0 88.6662 -87.7881>}]
 dyn_scarfFlapA_Base [{shapeType=Box, extentX=1.01, extentY=1.01, extentZ=1.01, rotation=<rot 0 -90 0>}]
 dyn_scarfFlapB_01 [{shapeType=Box, extentX=2.0506, extentY=9.3643, extentZ=6.8838, center=<4.3278 -0.2591 0.0548>, rotation=<rot -5.9389 -90 4.5552>}]
 dyn_scarfFlapB_01_Base [{shapeType=Box, extentX=1.01, extentY=1.01, extentZ=1.01, rotation=<rot 0 -90 0>}]
 dyn_scarfFlapB_02 [{shapeType=Box, extentX=1.5576, extentY=5.6898, extentZ=10.2175, center=<4.2829 0.0522 0.1467>, rotation=<rot 0 -87.4212 -87.8815>}]
 spine_02 [{shapeType=Capsule, length=5.8049, radius=12.6389, center=<0 -0.8267 0>, rotation=<rot 90 0 0>}]
}
constraints [
 {
  bone1Name=dyn_armDring_r
  bone2Name=dyn_armDring_Base_r
  swing1LimitDegrees=45
  swing2LimitDegrees=2
  swing2Motion=Locked
  twistLimitDegrees=2
  twistMotion=Locked
 }
 {
  bone1Name=dyn_armDring_l
  bone2Name=dyn_armDring_Base_l
  swing1LimitDegrees=45
  swing2LimitDegrees=1
  swing2Motion=Locked
  twistLimitDegrees=1
  twistMotion=Locked
 }
 {bone1Name=dyn_bandolier, bone2Name=dyn_bandolier_Base, swing1LimitDegrees=40, swing2LimitDegrees=5, twistLimitDegrees=2}
 {bone1Name=dyn_pouchA_l, bone2Name=dyn_pouchA_Base_l, swing1LimitDegrees=5, swing2LimitDegrees=25, twistLimitDegrees=5}
 {bone1Name=dyn_pouchB_l, bone2Name=dyn_pouchB_Base_l, swing1LimitDegrees=5, swing2LimitDegrees=25, twistLimitDegrees=5}
 {bone1Name=dyn_pouchC_l, bone2Name=dyn_pouchC_Base_l, swing1LimitDegrees=20, swing2LimitDegrees=5, twistLimitDegrees=5, twistMotion=Locked}
 {bone1Name=dyn_pouch_r, bone2Name=dyn_pouch_Base_r, swing1LimitDegrees=5, swing2LimitDegrees=25, twistLimitDegrees=5}
 {bone1Name=dyn_scarfFlapA, bone2Name=dyn_scarfFlapA_Base, swing1LimitDegrees=30, swing2LimitDegrees=2, twistLimitDegrees=2}
 {bone1Name=dyn_scarfFlapB_01, bone2Name=dyn_scarfFlapB_01_Base, swing1LimitDegrees=30, swing2LimitDegrees=2, twistLimitDegrees=2}
 {bone1Name=dyn_scarfFlapB_02, bone2Name=dyn_scarfFlapB_01, swing1LimitDegrees=30, swing2LimitDegrees=2, twistLimitDegrees=2}
 {
  bone1Name=dyn_armDring_Base_l
  bone2Name=spine_02
  swing1LimitDegrees=45
  swing1Motion=Free
  swing2LimitDegrees=45
  swing2Motion=Free
  twistLimitDegrees=45
  twistMotion=Free
 }
 {
  bone1Name=dyn_armDring_Base_r
  bone2Name=spine_02
  swing1LimitDegrees=45
  swing1Motion=Free
  swing2LimitDegrees=45
  swing2Motion=Free
  twistLimitDegrees=45
  twistMotion=Free
 }
 {
  bone1Name=dyn_scarfFlapA_Base
  bone2Name=spine_02
  swing1LimitDegrees=45
  swing1Motion=Free
  swing2LimitDegrees=45
  swing2Motion=Free
  twistLimitDegrees=45
  twistMotion=Free
 }
 {
  bone1Name=dyn_scarfFlapB_01_Base
  bone2Name=spine_02
  swing1LimitDegrees=45
  swing1Motion=Free
  swing2LimitDegrees=45
  swing2Motion=Free
  twistLimitDegrees=45
  twistMotion=Free
 }
 {
  bone1Name=dyn_pouchA_Base_l
  bone2Name=spine_02
  swing1LimitDegrees=45
  swing1Motion=Free
  swing2LimitDegrees=45
  swing2Motion=Free
  twistLimitDegrees=45
  twistMotion=Free
 }
 {
  bone1Name=dyn_pouchB_Base_l
  bone2Name=spine_02
  swing1LimitDegrees=45
  swing1Motion=Free
  swing2LimitDegrees=45
  swing2Motion=Free
  twistLimitDegrees=45
  twistMotion=Free
 }
 {
  bone1Name=dyn_pouchC_Base_l
  bone2Name=spine_02
  swing1LimitDegrees=45
  swing1Motion=Free
  swing2LimitDegrees=45
  swing2Motion=Free
  twistLimitDegrees=45
  twistMotion=Free
 }
 {
  bone1Name=dyn_pouch_Base_r
  bone2Name=spine_02
  swing1LimitDegrees=45
  swing1Motion=Free
  swing2LimitDegrees=45
  swing2Motion=Free
  twistLimitDegrees=45
  twistMotion=Free
 }
 {
  bone1Name=dyn_bandolier_Base
  bone2Name=spine_02
  swing1LimitDegrees=45
  swing1Motion=Free
  swing2LimitDegrees=45
  swing2Motion=Free
  twistLimitDegrees=45
  twistMotion=Free
 }
]
