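# Post-style AnimBP for the male medium UrbanScavenger body, adding secondary motion to bandanna, arm D-rings, strings and pouch.
# It copies the attached parent's pose, then weighted constraints position the dyn_ bones from nearby body bones.
# Head rotation relative to spine_05, remapped in the EventGraph, drives corrective bandanna offsets; headSwingDeltaBack and headPitchAim are set but unused here.
# A fixed pouch offset precedes the final RigidBody simulation, which uses an override physics asset in root-bone space.
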AnimBlueprint /BRCosmetics/Characters/Player/Male/Medium/Bodies/M_MED_UrbanScavenger/Meshes/M_MED_UrbanScavenger_AnimBP
target_skeleton=&/M_MED_UrbanScavenger_Skeleton
graphs [
 AnimGraph {nodes [
   n1 = "Constraint - dyn_armDring_Base_l (PPPP)"!(
    ComponentPose=Local To Component(LocalPose=Copy Pose From Mesh() {bUseAttachedParent=true})
    Alpha=1
    ConstraintWeights_0=0.0954
    ConstraintWeights_1:float=0.027
    ConstraintWeights_2:float=0.4599
    ConstraintWeights_3:float=0.4176
   ) -> (Pose) {
    alphaBoolBlend {blendOption=Linear}
    boneToModify {boneName=dyn_armDring_Base_l}
    constraintSetup [
     {targetBone {boneName=lowerarm_l}}
     {targetBone {boneName=hand_l}}
     {targetBone {boneName=lowerarm_twist_01_l}}
     {targetBone {boneName=lowerarm_twist_02_l}}
    ]
    constraintWeights [1,1,1,1]
   }
   n2 = "Constraint - dyn_armDring_Base_r (PPPP)"!(
    ComponentPose=@n1
    Alpha=1
    ConstraintWeights_0=0.0954
    ConstraintWeights_1:float=0.027
    ConstraintWeights_2:float=0.4599
    ConstraintWeights_3:float=0.4176
   ) -> (Pose) {
    alphaBoolBlend {blendOption=Linear}
    boneToModify {boneName=dyn_armDring_Base_r}
    constraintSetup [
     {targetBone {boneName=lowerarm_r}}
     {targetBone {boneName=hand_r}}
     {targetBone {boneName=lowerarm_twist_01_r}}
     {targetBone {boneName=lowerarm_twist_02_r}}
    ]
    constraintWeights [1,1,1,1]
   }
   n3 = "Constraint - dyn_bandanna_Base_r (PPPPPPPP)"!(
    ComponentPose=@n2
    Alpha=1
    ConstraintWeights_0=0.0963
    ConstraintWeights_1:float=0.0261
    ConstraintWeights_2:float=0.2142
    ConstraintWeights_3:float=0.0328
    ConstraintWeights_4:float=0.0793
    ConstraintWeights_5:float=0.0177
    ConstraintWeights_6:float=0.0774
    ConstraintWeights_7:float=0.6497
   ) -> (Pose) {
    alphaBoolBlend {blendOption=Linear}
    boneToModify {boneName=dyn_bandanna_Base_r}
    constraintSetup [
     {targetBone {boneName=head}}
     {targetBone {boneName=clavicle_l}}
     {targetBone {boneName=clavicle_r}}
     {targetBone {boneName=neck_01}}
     {targetBone {boneName=neck_02}}
     {targetBone {boneName=spine_03}}
     {targetBone {boneName=spine_04}}
     {targetBone {boneName=spine_05}}
    ]
    constraintWeights [1,1,1,1,1,1,1,1]
   }
   n4 = "Constraint - dyn_bandanna_Base_l (PPPPPPPP)"!(
    ComponentPose=@n3
    Alpha=1
    ConstraintWeights_0=0.1428
    ConstraintWeights_1:float=0.144
    ConstraintWeights_2:float=0.0397
    ConstraintWeights_3:float=0.0411
    ConstraintWeights_4:float=0.0887
    ConstraintWeights_5:float=0.0129
    ConstraintWeights_6:float=0.0845
    ConstraintWeights_7:float=0.4463
   ) -> (Pose) {
    alphaBoolBlend {blendOption=Linear}
    boneToModify {boneName=dyn_bandanna_Base_l}
    constraintSetup [
     {targetBone {boneName=head}}
     {targetBone {boneName=clavicle_l}}
     {targetBone {boneName=clavicle_r}}
     {targetBone {boneName=neck_01}}
     {targetBone {boneName=neck_02}}
     {targetBone {boneName=spine_03}}
     {targetBone {boneName=spine_04}}
     {targetBone {boneName=spine_05}}
    ]
    constraintWeights [1,1,1,1,1,1,1,1]
   }
   n5 = "Constraint - dyn_bandannaFront_Base (PPPPPPPP)"!(
    ComponentPose=@n4
    Alpha=1
    ConstraintWeights_0=0.0676
    ConstraintWeights_1:float=0.0612
    ConstraintWeights_2:float=0.0879
    ConstraintWeights_3:float=0.0367
    ConstraintWeights_4:float=0.0762
    ConstraintWeights_5:float=0.0243
    ConstraintWeights_6:float=0.1729
    ConstraintWeights_7:float=0.4732
   ) -> (Pose) {
    alphaBoolBlend {blendOption=Linear}
    boneToModify {boneName=dyn_bandannaFront_Base}
    constraintSetup [
     {targetBone {boneName=head}}
     {targetBone {boneName=clavicle_l}}
     {targetBone {boneName=clavicle_r}}
     {targetBone {boneName=neck_01}}
     {targetBone {boneName=neck_02}}
     {targetBone {boneName=spine_03}}
     {targetBone {boneName=spine_04}}
     {targetBone {boneName=spine_05}}
    ]
    constraintWeights [1,1,1,1,1,1,1,1]
   }
   n6 = "Transform (Modify) Bone - Bone: dyn_bandanna_r"(
    ComponentPose=@n5
    Alpha=Get headTwist_r() -> (headTwist_r:double)
    Translation="-6.021,-3.516,-5.200"
    Rotation="20.732,-37.365,-3.521"
    Scale="1,1,1"
   ) {
    rotationMode=BMM_Additive
    rotationSpace=BCS_BoneSpace
    translationMode=BMM_Additive
    translationSpace=BCS_BoneSpace
    alphaBoolBlend {blendOption=Linear}
    boneToModify {boneName=dyn_bandanna_r}
    rotation=<rot 20.732 -37.3649 -3.521>
    translation=<-6.021 -3.516 -5.2>
   }
   n7 = "Transform (Modify) Bone - Bone: dyn_bandanna_l"(
    ComponentPose=@n6
    Alpha=Get headTwist_l() -> (headTwist_l:double)
    Translation="-2.132,-5.204,-1.487"
    Rotation="0.000,-17.305,-9.170"
    Scale="1,1,1"
   ) {
    rotationMode=BMM_Additive
    rotationSpace=BCS_BoneSpace
    translationMode=BMM_Additive
    translationSpace=BCS_BoneSpace
    alphaBoolBlend {blendOption=Linear}
    alphaScaleBiasClamp {bMapRange=true, inRange {max=0.85}, outRange {min=0.2}}
    boneToModify {boneName=dyn_bandanna_l}
    rotation {roll=-9.1701, yaw=-17.3051}
    translation=<-2.132 -5.204 -1.487>
   }
   n8 = "Transform (Modify) Bone - Bone: dyn_bandannaFront"(
    ComponentPose=@n7
    Alpha=Get headSwingDeltaFront() -> (headSwingDeltaFront:double)
    Translation="0.000,-2.082,0.000"
    Rotation="0.000,-14.054,0.000"
    Scale="1.000,1.000,1.5"
   ) {
    rotationMode=BMM_Additive
    rotationSpace=BCS_BoneSpace
    scaleMode=BMM_Replace
    scaleSpace=BCS_BoneSpace
    translationMode=BMM_Additive
    translationSpace=BCS_BoneSpace
    alphaBoolBlend {blendOption=Linear}
    boneToModify {boneName=dyn_bandannaFront}
    rotation {yaw=-14.0541}
    scale {z=1.5}
    translation {y=-2.082}
   }
   n9 = "Constraint - dyn_string_Driven_r (PPPPPPPP)"!(
    ComponentPose=@n8
    Alpha=1
    ConstraintWeights_0=0.3132
    ConstraintWeights_1:float=0.3909
    ConstraintWeights_2:float=0.2586
    ConstraintWeights_3:float=0.0133
    ConstraintWeights_4:float=0.0044
    ConstraintWeights_5:float=0.0152
    ConstraintWeights_6:float=0.0026
    ConstraintWeights_7:float=0.0018
   ) -> (Pose) {
    alphaBoolBlend {blendOption=Linear}
    boneToModify {boneName=dyn_string_Driven_r}
    constraintSetup [
     {targetBone {boneName=pelvis}}
     {targetBone {boneName=thigh_twist_01_r}}
     {targetBone {boneName=thigh_twist_01_l}}
     {targetBone {boneName=thigh_r}}
     {targetBone {boneName=thigh_l}}
     {targetBone {boneName=spine_01}}
     {targetBone {boneName=spine_02}}
     {targetBone {boneName=spine_03}}
    ]
    constraintWeights [1,1,1,1,1,1,1,1]
   }
   n10 = "Constraint - dyn_string_Driven_l (PPPPPPPP)"!(
    ComponentPose=@n9
    Alpha=1
    ConstraintWeights_0=0.3359
    ConstraintWeights_1:float=0.3612
    ConstraintWeights_2:float=0.2692
    ConstraintWeights_3:float=0.0069
    ConstraintWeights_4:float=0.0059
    ConstraintWeights_5:float=0.0163
    ConstraintWeights_6:float=0.0027
    ConstraintWeights_7:float=0.0019
   ) -> (Pose) {
    alphaBoolBlend {blendOption=Linear}
    boneToModify {boneName=dyn_string_Driven_l}
    constraintSetup [
     {targetBone {boneName=pelvis}}
     {targetBone {boneName=thigh_twist_01_r}}
     {targetBone {boneName=thigh_twist_01_l}}
     {targetBone {boneName=thigh_r}}
     {targetBone {boneName=thigh_l}}
     {targetBone {boneName=spine_01}}
     {targetBone {boneName=spine_02}}
     {targetBone {boneName=spine_03}}
    ]
    constraintWeights [1,1,1,1,1,1,1,1]
   }
   n11 = "Constraint - dyn_pouchC_Base_l (PPPPPP)"!(
    ComponentPose=@n10
    Alpha=1
    ConstraintWeights_0=0.1007
    ConstraintWeights_1:float=0.2063
    ConstraintWeights_2:float=0.1931
    ConstraintWeights_3:float=0.3696
    ConstraintWeights_4:float=0.0297
    ConstraintWeights_5:float=0.1006
   ) -> (Pose) {
    alphaBoolBlend {blendOption=Linear}
    boneToModify {boneName=dyn_pouchC_Base_l}
    constraintSetup [
     {targetBone {boneName=clavicle_l}}
     {targetBone {boneName=spine_01}}
     {targetBone {boneName=spine_02}}
     {targetBone {boneName=spine_03}}
     {targetBone {boneName=spine_04}}
     {targetBone {boneName=pelvis}}
    ]
    constraintWeights [1,1,1,1,1,1]
   }
   n12 = "Transform (Modify) Bone - Bone: dyn_pouchC_Base_l"(ComponentPose=@n11, Alpha=1, Translation="-0.041,0.259,-1.512", Rotation="2.722,11.212,-13.491", Scale="1,1,1") {
    rotationMode=BMM_Additive
    rotationSpace=BCS_BoneSpace
    translationMode=BMM_Additive
    translationSpace=BCS_BoneSpace
    alphaBoolBlend {blendOption=Linear}
    boneToModify {boneName=dyn_pouchC_Base_l}
    rotation=<rot 2.722 11.2122 -13.4912>
    translation=<-0.041 0.2593 -1.5119>
   }
   n13 = RigidBody(ComponentPose=@n12) {
    bClampLinearTranslationLimitToRefPose=true
    bOverrideWorldGravity=true
    overridePhysicsAsset=&/M_MED_UrbanScavenger_Physics
    simulationSpace=BaseBoneSpace
    alphaBoolBlend {blendOption=Linear}
    baseBoneRef {boneName=root}
    componentAppliedLinearAccClamp=<5000 5000 5000>
    componentLinearAccScale=<0.5 0.5 0.5>
    componentLinearVelScale=<0.3 0.3 0.3>
    overrideWorldGravity {z=-980}
   }
   Output Pose(Result=Component To Local(ComponentPose=@n13)) {layerGroup=None}
  ]}
 EventGraph {nodes [
   Event BlueprintUpdateAnimation()
   > Sequence()
   n1 = BreakTransform(InTransform=GetDeltaTransformFromRefPose(self=GetOwnerSkeletalMeshComponent(), BoneName=head, BaseName=spine_05))
   n2 = BreakRotator(InRot=@n1.Rotation)
   > Set headSwingDeltaBack(headSwingDeltaBack:double=MapRangeClamped(Value=@n2.Yaw, InRangeA=-3.5, InRangeB=-47)) -> (Output_Get:double)
   > Set headSwingDeltaFront(headSwingDeltaFront:double=MapRangeClamped(Value=@n2.Yaw, InRangeA=-3.5, InRangeB=27)) -> (Output_Get:double)
   n3 = BreakRotator(InRot=@n1.Rotation)
   > Set headTwist_r(headTwist_r:double=MapRangeClamped(Value=@n3, InRangeA=-10, InRangeB=-60)) -> (Output_Get:double)
   > Set headPitchAim(headPitchAim:double=MapRangeClamped(Value=@n3.Pitch, InRangeA=0, InRangeB=37)) -> (Output_Get:double)
   > Set headTwist_l(headTwist_l:double=MapRangeClamped(Value=@n3, InRangeA=10, InRangeB=60)) -> (Output_Get:double)
   TryGetPawnOwner()
  ]}
]
variables [
 headPitchAim {type=number}
 headSwingDeltaBack {type=number}
 headSwingDeltaFront {type=number}
 headTwist_l {type=number}
 headTwist_r {type=number}
]
