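# Physics asset for the face accessory mesh of the male medium Urban Scavenger player body, covering a cap strap and hair strands.
# Capsule bodies sit on the head, the cap strap and left and right hair bones.
# The strap and first bone of each strand are constrained to the head, and second hair bones, where present, to their first bone.
# Every joint shares the same narrow swing and twist limits, keeping secondary motion tight.
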PhysicsAsset /BRCosmetics/Characters/Player/Male/Medium/Bodies/M_MED_UrbanScavenger/Meshes/Parts/M_MED_UrbanScavenger_FaceAcc_Physics
bodies {
 dyn_capStrap [{shapeType=Capsule, length=0.9239, radius=0.7328, center=<0.509 0 0.0001>, rotation=<rot 90 0 0>}]
 dyn_hairA_01_l [{shapeType=Capsule, length=2.3459, radius=1.1425, center=<0.9858 -0.311 -0.0336>, rotation=<rot 90 0 0>}]
 dyn_hairA_01_r [{shapeType=Capsule, length=1.8068, radius=1.2256, center=<1.009 -0.1127 0.2039>, rotation=<rot 90 0 0>}]
 dyn_hairA_02_l [{shapeType=Capsule, length=1.9359, radius=1.0832, center=<0.912 0.0805 -0.2>, rotation=<rot 0 -77.04 -76.884>}]
 dyn_hairA_02_r [{shapeType=Capsule, length=2.1997, radius=0.9003, center=<1.2211 0.4698 -0.0919>, rotation=<rot 0 -62.58 -78.2528>}]
 dyn_hairB_01_r [{shapeType=Capsule, length=2.2765, radius=0.9377, center=<1.4396 -0.1906 -0.0425>, rotation=<rot 90 0 0>}]
 dyn_hairB_02_r [{shapeType=Capsule, length=1.7677, radius=1.2531, center=<1.433 0.3747 -0.1358>, rotation=<rot 0 -68.2854 -89.2025>}]
 dyn_hairC_01_r [{shapeType=Capsule, length=2.3955, radius=0.9128, center=<1.2444 -0.0519 0.0084>, rotation=<rot 0 91.0373 -89.4261>}]
 head [{shapeType=Capsule, length=13.2863, radius=5.3557, center=<3.604 4.2162 0>, rotation=<rot 90 0 0>}]
}
constraints [
 {bone1Name=dyn_hairA_02_r, bone2Name=dyn_hairA_01_r, swing1LimitDegrees=20, swing2LimitDegrees=20, twistLimitDegrees=12}
 {bone1Name=dyn_hairB_02_r, bone2Name=dyn_hairB_01_r, swing1LimitDegrees=20, swing2LimitDegrees=20, twistLimitDegrees=12}
 {bone1Name=dyn_hairA_02_l, bone2Name=dyn_hairA_01_l, swing1LimitDegrees=20, swing2LimitDegrees=20, twistLimitDegrees=12}
 {bone1Name=dyn_hairA_01_r, bone2Name=head, swing1LimitDegrees=20, swing2LimitDegrees=20, twistLimitDegrees=12}
 {bone1Name=dyn_hairB_01_r, bone2Name=head, swing1LimitDegrees=20, swing2LimitDegrees=20, twistLimitDegrees=12}
 {bone1Name=dyn_hairC_01_r, bone2Name=head, swing1LimitDegrees=20, swing2LimitDegrees=20, twistLimitDegrees=12}
 {bone1Name=dyn_capStrap, bone2Name=head, swing1LimitDegrees=20, swing2LimitDegrees=20, twistLimitDegrees=12}
 {bone1Name=dyn_hairA_01_l, bone2Name=head, swing1LimitDegrees=20, swing2LimitDegrees=20, twistLimitDegrees=12}
]
